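; Base toy basketball projectile that records its summon serial number and first hit component, and animates into the goal when it scores.
; Scoring stores lerp start and end positions, with an optional entry point that first plays a cue indexed by score.
; The animation disables projectile simulation and collision, then a timeline moves and scales the ball while authority silently kills it after the lerp duration.
; The end position's OnRep also triggers the animation.
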(event Visuals|EventInitializeToyInstance
  (params (OwningPC :type Fort%20Player%20Controller%20Object%20Reference) (NumTimesSummoned :type Integer))
  (|CallParentFunction OwningPC NumTimesSummoned)
  (Variables|Default|SetBallSerialNumber NumTimesSummoned))

(event Custom|PlayScoringAnim
  (bind (_returnvalue_location _returnvalue_rotation) (Math|Transform|BreakTransform (Transformation|GetWorldTransform (Variables|FortPhysicsBall|GetSphereCollision))))
  (Variables|/Script/Engine.ProjectileMovementComponent|SetbSimulationEnabled false (Variables|Default|GetProjectileMovement))
  (Collision|SetActorEnableCollision false)
  (Utilities|FlowControl|Sequence
    (:then_0
      (Networking|SwitchHasAuthority
        (:Authority
          (Utilities|FlowControl|Delay (Variables|Default|GetLerpDuration))
          (Gameplay|SilentDie false))))
    (:then_1
      (Components|Timeline|SetPlayRate (Variables|B_ToyBasketball_Base|GetLerpIntoGoalTimeline) (/ 1.0 (Variables|Default|GetLerpDuration)))
      (bind (_direction _position _scale) (Timeline|LerpIntoGoalTimeline PlayFromStart)
        (:Update
          (Transformation|SetWorldTransform (Variables|FortPhysicsBall|GetSphereCollision) (Math|Transform|MakeTransform (Math|Vector|Lerp(Vector) (Variables|Default|GetIdealLerpStartPos) (Variables|Default|GetIdealLerpEndPos) _position) _returnvalue_rotation (Math|Conversions|ToVector(Float) _scale)) false false))))))

(event EventOnHitTarget
  (params (StartPos :type Vector :default "0, 0, 0") (EndPos :type Vector :default "0, 0, 0"))
  (Variables|Default|SetIdealLerpStartPos StartPos)
  (Variables|Default|SetIdealLerpEndPos EndPos)
  (CallFunction|PlayScoringAnim))

(event OnComponentHit(SphereCollision)
  (params (HitComponent :type Primitive%20Component%20Object%20Reference) (OtherActor :type Actor%20Object%20Reference) (OtherComp :type Primitive%20Component%20Object%20Reference) (NormalImpulse :type Vector :default "0, 0, 0") (Hit :type Hit%20Result%20Structure%20%28by%20ref%29))
  (Utilities|IsValid (Variables|Default|GetHitComponent)
    (:Is%20Not%20Valid
      (Variables|Default|SetHitComponent OtherComp)
      (Networking|ForceNetUpdate))) ; Only track the first collision
)

(fn OnRep_IdealLerpEndPos
  (CallFunction|PlayScoringAnim))

(fn PlayScoringCueAndDestruct
  (params (ScoreAmount :type Integer) (TargetStartPos :type Vector) (TargetEndPos :type Vector))
  (bind _output_get (LocalVariables|SetCueToPlay (Utilities|Array|Get(acopy) (Variables|Default|GetCuesForScoreValues) (select (Utilities|Array|IsValidIndex (Variables|Default|GetCuesForScoreValues) ScoreAmount) ScoreAmount 0))))
  (CallFunction|PlayCueCenteredonMe _output_get)
  (Variables|Default|SetIdealLerpStartPos TargetStartPos)
  (Variables|Default|SetIdealLerpEndPos TargetEndPos)
  (CallFunction|PlayScoringAnim))

(fn ConstructionScript
  (|CallParentFunction))
